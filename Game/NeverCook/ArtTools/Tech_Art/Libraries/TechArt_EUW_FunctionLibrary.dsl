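; Tech-art Blueprint function library, named for Editor Utility Widgets, with helpers for sorting, asset lookup and panel reordering.
; Int and float bubble sorts swap an index array alongside values; float supports reverse order and prints on each swap.
; String sort bubbles pairs through a recursive character-code comparison that leaves prefix matches unswapped.
; Other helpers fetch asset data from an actor's class package and reorder panel children by re-adding them.
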(fn BubbleSort_FloatLogic
  (params (Value%20Array%20%28By%20Ref%29 :type Array%20of%20Float%20%28double-precision%29s) (Index%20Array%20%28By%20Ref%29 :type Array%20of%20Integers) (Index :type Integer) (Index+1 :type Integer) (ReverseOrder :type Boolean%20%28by%20ref%29))
  (bind _output (Utilities|Array|Get(aref) Value%20Array%20%28By%20Ref%29 Index))
  (bind _output_2 (Utilities|Array|Get(aref) Value%20Array%20%28By%20Ref%29 Index+1))
  (if (select ReverseOrder (>= _output _output_2) (<= _output _output_2))
    (return)
    (else
      (Development|PrintString 0)
      (Utilities|FlowControl|Sequence
        (:then_0
          (LocalVariables|SetLocalValueSwap _output)
          (LocalVariables|SetLocalIndexSwap (Utilities|Array|Get(acopy) Index%20Array%20%28By%20Ref%29 Index)))
        (:then_1
          (Utilities|Array|SetArrayElem Value%20Array%20%28By%20Ref%29 Index _output_2)
          (Utilities|Array|SetArrayElem Index%20Array%20%28By%20Ref%29 Index (Utilities|Array|Get(acopy) Index%20Array%20%28By%20Ref%29 Index+1)))
        (:then_2
          (Utilities|Array|SetArrayElem Value%20Array%20%28By%20Ref%29 Index+1 (LocalVariables|GetLocalValueSwap))
          (Utilities|Array|SetArrayElem Index%20Array%20%28By%20Ref%29 Index+1 (LocalVariables|GetLocalIndexSwap)))
        (:then_3
          (return))))))

(fn BubbleSort_IntLogic
  (params (Value%20Array%20%28By%20Ref%29 :type Array%20of%20Integers) (Index%20Array%20%28By%20Ref%29 :type Array%20of%20Integers) (Index :type Integer) (Index+1 :type Integer))
  (bind _output (Utilities|Array|Get(acopy) Value%20Array%20%28By%20Ref%29 Index+1))
  (bind _output_2 (Utilities|Array|Get(acopy) Value%20Array%20%28By%20Ref%29 Index))
  (if (<= _output_2 _output)
    (return)
    (else
      (Utilities|FlowControl|Sequence
        (:then_0
          (LocalVariables|SetLocalValueSwap _output_2)
          (LocalVariables|SetLocalIndexSwap (Utilities|Array|Get(acopy) Index%20Array%20%28By%20Ref%29 Index)))
        (:then_1
          (Utilities|Array|SetArrayElem Value%20Array%20%28By%20Ref%29 Index _output)
          (Utilities|Array|SetArrayElem Index%20Array%20%28By%20Ref%29 Index (Utilities|Array|Get(acopy) Index%20Array%20%28By%20Ref%29 Index+1)))
        (:then_2
          (Utilities|Array|SetArrayElem Value%20Array%20%28By%20Ref%29 Index+1 (LocalVariables|GetLocalValueSwap))
          (Utilities|Array|SetArrayElem Index%20Array%20%28By%20Ref%29 Index+1 (LocalVariables|GetLocalIndexSwap)))
        (:then_3
          (return))))))

(fn CompareStrings_Logic
  (params (FirstString :type String) (SecondString :type String) (Recursive%20Count :type Integer))
  (bind _returnvalue (Utilities|String|GetCharacterArrayfromString SecondString))
  (bind _returnvalue_2 (Utilities|String|GetCharacterArrayfromString FirstString))
  (bind _output (Utilities|Array|Get(acopy) _returnvalue Recursive%20Count))
  (bind _returnvalue_3 (Utilities|String|GetCharacterasNumber _output))
  (bind _output_2 (Utilities|Array|Get(acopy) _returnvalue_2 Recursive%20Count))
  (bind _returnvalue_4 (Utilities|String|GetCharacterasNumber _output_2))
  ;[1 1 1] Deal with length difference on strings that are identical
  (if (and (> (Utilities|Array|Length _returnvalue_2) Recursive%20Count) (> (Utilities|Array|Length _returnvalue) Recursive%20Count))
    (if (< _returnvalue_4 _returnvalue_3)
      (return FirstString SecondString)
      (elif (== _returnvalue_4 _returnvalue_3)
        (bind (_sortedfirststring _sortedsecondstring) (StringSortLogic|CompareStringsLogic FirstString SecondString (+ Recursive%20Count 1)))
        (return _sortedfirststring _sortedsecondstring)
        (else
          (return SecondString FirstString))))
    (else
      (return FirstString SecondString))))

(fn GetContentAssetPathFromActorRef
  (params (Actor :type Object%20Reference))
  (bind _outassetdata (AssetRegistry|GetAssetsbyPackageName (AssetRegistry|GetAssetRegistry) (Utilities|String|StringToName (Utilities|String|Split (Utilities|ToString(SoftClassReference) (Utilities|ResolveSoftReference (Utilities|GetClass Actor))) "." "IgnoreCase" "FromEnd")) false true))
  (return (Utilities|Array|Get(acopy) _outassetdata) (Utilities|IsArrayNonEmpty _outassetdata)))

(fn TechArtEUWIntBubbleSortSortOnce
  (params (Input%20Value%20Array :type Array%20of%20Integers) (CurrentPass :type Integer))
  (bind _returnvalue (Utilities|Array|Length Input%20Value%20Array))
  (bind _sortedindexs (LocalVariables|GetSortedIndexs))
  (LocalVariables|SetSortedValues Input%20Value%20Array)
  (if (> _returnvalue 1)
    (Utilities|Array|Resize _sortedindexs _returnvalue)
    (for (_ _array_index) _sortedindexs
      (Utilities|Array|SetArrayElem _sortedindexs _array_index _array_index))
    (for _index (range (+ (- (- _returnvalue 2) CurrentPass) 1))
      (BubbleSortLogics|BubbleSortIntLogic (LocalVariables|GetSortedValues) (LocalVariables|GetSortedIndexs) _index (+ _index 1)))
    (return (LocalVariables|GetSortedValues) (LocalVariables|GetSortedIndexs))
    (else
      (return Input%20Value%20Array (Utilities|Array|MakeArray 0)))))

(fn TechArt_EUW_FloatBubbleSort
  (params (Input%20Value%20Array :type Array%20of%20Float%20%28double-precision%29s) (ReverseOrder :type Boolean%20%28by%20ref%29))
  (bind _returnvalue (Utilities|Array|Length Input%20Value%20Array))
  (bind _returnvalue_2 (- _returnvalue 2))
  (bind _sortedindexs (LocalVariables|GetSortedIndexs))
  (LocalVariables|SetSortedValues Input%20Value%20Array)
  (if (> _returnvalue 1)
    (Utilities|Array|Resize _sortedindexs _returnvalue)
    (for (_ _array_index) _sortedindexs
      (Utilities|Array|SetArrayElem _sortedindexs _array_index _array_index))
    (for _index (range (+ _returnvalue_2 1))
      (for _index_2 (range (+ (- _returnvalue_2 _index) 1))
        (BubbleSortLogics|BubbleSortFloatLogic (LocalVariables|GetSortedValues) (LocalVariables|GetSortedIndexs) _index_2 (+ _index_2 1) ReverseOrder)))
    (return (LocalVariables|GetSortedValues) (LocalVariables|GetSortedIndexs))
    (else
      (return Input%20Value%20Array (Utilities|Array|MakeArray 0)))))

(fn TechArt_EUW_IntBubbleSort
  (params (Input%20Value%20Array :type Array%20of%20Integers))
  (bind _returnvalue (Utilities|Array|Length Input%20Value%20Array))
  (bind _returnvalue_2 (- _returnvalue 2))
  (bind _sortedindexs (LocalVariables|GetSortedIndexs))
  (LocalVariables|SetSortedValues Input%20Value%20Array)
  (if (> _returnvalue 1)
    (Utilities|Array|Resize _sortedindexs _returnvalue)
    (for (_ _array_index) _sortedindexs
      (Utilities|Array|SetArrayElem _sortedindexs _array_index _array_index))
    (for _index (range (+ _returnvalue_2 1))
      (for _index_2 (range (+ (- _returnvalue_2 _index) 1))
        (BubbleSortLogics|BubbleSortIntLogic (LocalVariables|GetSortedValues) (LocalVariables|GetSortedIndexs) _index_2 (+ _index_2 1))))
    (return (LocalVariables|GetSortedValues) (LocalVariables|GetSortedIndexs))
    (else
      (return Input%20Value%20Array (Utilities|Array|MakeArray 0)))))

(fn TechArt_EUW_SortStringArrayAlphabetically
  (params (InStringArray :type Array%20of%20Strings))
  (bind _returnarray (LocalVariables|GetReturnArray))
  (bind _output_get (LocalVariables|SetReturnArray InStringArray))
  (bind _arrayindex (Utilities|Array|ReverseforEachLoop _output_get)
    (:LoopBody
      (for _index (range _arrayindex)
        (bind (_sortedfirststring _sortedsecondstring) (StringSortLogic|CompareStringsLogic (Utilities|Array|Get(acopy) _output_get _index) (Utilities|Array|Get(acopy) _output_get (+ _index 1))))
        (Utilities|Array|SetArrayElem _returnarray _index _sortedfirststring)
        (Utilities|Array|SetArrayElem _returnarray (+ _index 1) _sortedsecondstring)))
    (:Completed
      (return (LocalVariables|GetReturnArray)))))

(fn TechArtEUW_ReorderWidgetPanel
  (params (ParentPanelComponent :type Panel%20Widget%20Object%20Reference%20%28by%20ref%29) (ChildComponent :type Widget%20Object%20Reference%20%28by%20ref%29) (NewPosition :type Integer))
  (bind _output_get (LocalVariables|SetLocalChildren (Widget|Panel|GetAllChildren ParentPanelComponent)))
  (Utilities|Array|RemoveItem _output_get ChildComponent)
  (Utilities|Array|Insert _output_get ChildComponent NewPosition)
  (for _array_element _output_get
    (Widget|Panel|AddChild ParentPanelComponent _array_element))
  (return))
