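# Editor utility function library that exposes the project's Python techart module to Blueprint tools.
# Functions cover Blueprint and actor subobjects, selected components, actor source assets, recursive dependencies, and reading, saving to or removing from collections.
# Each runs an Execute Python Script node that reloads techart before calling it.
# Collection save and remove use the object array, falling back to asset data when it is empty; reading can optionally also return objects.
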EditorUtilityBlueprint /Game/NeverCook/ArtTools/Tech_Art/Libraries/Techart_UEFunctionLib
graphs [
 GetAllBlueprintSubobjects {nodes [
   n1 = GetAllBlueprintSubobjects!() -> (then, Blueprint:Object Reference, ClassFilter:Object Class Reference) {is=title K2Node_FunctionEntry}
   > n2 = Execute Python Script!(
    execute
    PythonScript="import unreal\r\nimport techart\r\nimport importlib\r\nimportlib.reload(techart)\r\n\r\nresults=techart.get_all_blueprint_subobjects(blueprint,classFilter_input=classfilter)"
    blueprint:Object Reference=@n1
    classfilter:Object Class Reference=@n1.ClassFilter
   ) -> (then, ReturnValue, results:Array of Object References) {inputs [blueprint,classfilter], outputs [results]}
   > Return Node!(execute, Componenets:Array of Object References=@n2.results)
  ]}
 GetAllDependenciesRecursive {nodes [
   n1 = GetAllDependenciesRecursive!() -> (then, PackageName:Name, options:Asset Registry Dependency Options Structure, max_depth:int) {is=title K2Node_FunctionEntry}
   > n2 = Execute Python Script!(
    execute
    PythonScript="import unreal\r\nimport techart\r\nimport importlib\r\nimportlib.reload(techart)\r\n\r\nresults=techart.get_all_dependencies_recursive(object,options,max_depth)"
    object:Name=@n1
    options:Asset Registry Dependency Options Structure=@n1.options
    maxDepth:int=@n1.max_depth
   ) -> (then, ReturnValue, results:Array of Names) {inputs [object,options,maxDepth], outputs [results]}
   > Return Node!(execute, OutDependencies:Array of Names=@n2.results)
  ]}
 GetAllInstanceSubobjects {nodes [
   n1 = GetAllInstanceSubobjects!() -> (then, Actor:Actor Object Reference, Filter:Object Class Reference) {is=title K2Node_FunctionEntry}
   > n2 = Execute Python Script!(
    execute
    PythonScript="import unreal\r\nimport techart\r\nimport importlib\r\nimportlib.reload(techart)\r\n\r\nresults=techart.get_all_instance_subobjects(actor,classFilter_input=classfilter)"
    actor:Actor Object Reference=@n1
    classfilter:Object Class Reference=@n1.Filter
   ) -> (then, ReturnValue, results:Array of Object References) {inputs [actor,classfilter], outputs [results]}
   > Return Node!(execute, Components:Array of Object References=@n2.results)
  ]}
 GetAssetFromActor {nodes [
   n1 = GetAssetFromActor!() -> (then, actor:Actor Object Reference) {is=title K2Node_FunctionEntry}
   > n2 = Execute Python Script!(
    execute
    PythonScript="import unreal\r\nimport techart\r\nimport importlib\r\nimportlib.reload(techart)\r\n\r\nresults=techart.get_asset_from_actor(actor)"
    actor=@n1
   ) -> (then, ReturnValue, results:Object Reference) {inputs [actor], outputs [results]}
   > Return Node!(execute, asset:Object Reference=@n2.results)
  ]}
 GetAssetsInCollection {nodes [
   n1 = GetAssetsInCollection!() -> (then, Collection Name:String, LoadObjects?:bool=~) {is=title K2Node_FunctionEntry}
   > n2 = Branch(Condition=@n1.LoadObjects?)
   > n3 = Execute Python Script!(
    execute
    PythonScript="import unreal\r\nimport techart\r\nimport importlib\r\nimportlib.reload(techart)\r\n\r\nresults=techart.get_assets_in_collection(collectionname, returnType=0)"
    collectionname:String=@n1
   ) -> (then, ReturnValue, results:Array of Asset Data Structures) {inputs [collectionname], outputs [results]}
   > n4 = Execute Python Script(
    PythonScript="import unreal\r\nimport techart\r\nimport importlib\r\nimportlib.reload(techart)\r\n\r\nresults=techart.get_assets_in_collection(collectionname, returnType=1)"
    collectionname=@n1
   ) -> (results:Array of Object References) {inputs [collectionname], outputs [results]}
   > Return Node!(execute, AsObjects:Array of Object References=@n4.results, AsData:Array of Asset Data Structures=@n3.results)
   n5 = Execute Python Script(
    execute=@n2.else
    PythonScript="import unreal\r\nimport techart\r\nimport importlib\r\nimportlib.reload(techart)\r\n\r\nresults=techart.get_assets_in_collection(collectionname, returnType=0)"
    collectionname=@n1
   ) -> (results:Array of Asset Data Structures) {inputs [collectionname], outputs [results]}
   > Return Node(AsData=@n5.results)
  ]}
 GetSelectedActorsSelectedComponents {nodes [
   n1 = GetSelectedActorsSelectedComponents!() -> (then, FilterClass:Object Class Reference) {is=title K2Node_FunctionEntry}
   > n2 = Execute Python Script!(
    execute
    PythonScript="import unreal\r\nimport techart\r\nimport importlib\r\nimportlib.reload(techart)\r\n\r\nresults = techart.get_selected_actors_selected_components(classFilter_input=filter_in)"
    filter_in:Object Class Reference=@n1
   ) -> (then, ReturnValue, results:Array of Object References) {inputs [filter_in], outputs [results]}
   > Return Node!(execute, Components:Array of Object References=@n2.results)
  ]}
 RemoveAssetsFromCollection {nodes [
   n1 = RemoveAssetsFromCollection!() -> (then, Objects:Array of Object References, Data:Array of Asset Data Structures, Collection:String) {is=title K2Node_FunctionEntry}
   > n2 = Branch(Condition=Is Empty!(TargetArray:Array of Object References=@n1) -> (ReturnValue:bool))
   > n3 = Execute Python Script!(
    execute
    PythonScript="import unreal\r\nimport techart\r\nimport importlib\r\nimportlib.reload(techart)\r\n\r\nresults=techart.remove_assets_from_collection(datas, collection)"
    datas:Array of Asset Data Structures=@n1.Data
    Collection:String=@n1.Collection
   ) -> (then, ReturnValue, results:bool=~) {inputs [datas,Collection], outputs [results]}
   > Return Node!(execute, Success?:bool=@n3.results)
   n4 = Execute Python Script!(
    execute=@n2.else
    PythonScript="import unreal\r\nimport techart\r\nimport importlib\r\nimportlib.reload(techart)\r\n\r\nresults=techart.remove_assets_from_collection(objects, collection)"
    objects:Array of Object References=@n1
    Collection=@n1.Collection
   ) -> (then, ReturnValue, results=~) {inputs [objects,Collection], outputs [results]}
   > Return Node(Success?=@n4.results)
  ]}
 SaveAssetsToCollection {nodes [
   n1 = SaveAssetsToCollection!() -> (
    then
    Objects:Array of Object References
    AssetData:Array of Asset Data Structures
    Collection:String
    CollectionType:int
    Overwrite?:bool=~
    create?:bool=true
   ) {is=title K2Node_FunctionEntry}
   > n2 = Branch(Condition=Is Empty(TargetArray=@n1))
   > n3 = Execute Python Script!(
    execute
    PythonScript="import unreal\r\nimport techart\r\nimport importlib\r\nimportlib.reload(techart)\r\n\r\nresults=techart.save_assets_to_collection(data, collection, sharetype=sharetype_in,overwrite=overwrite_in,create=create_in)"
    Data:Array of Asset Data Structures=@n1.AssetData
    Collection=@n1.Collection
    sharetype_in:int=@n1.CollectionType
    overwrite_in:bool=@n1.Overwrite?
    create_in:bool=@n1.create?
   ) -> (then, ReturnValue, results=~) {inputs [Data,Collection,sharetype_in,overwrite_in,create_in], outputs [results]}
   > Return Node(Success?=@n3.results)
   n4 = Execute Python Script!(
    execute=@n2.else
    PythonScript="import unreal\r\nimport techart\r\nimport importlib\r\nimportlib.reload(techart)\r\n\r\nresults=techart.save_assets_to_collection(assets, collection, sharetype=sharetype_in,overwrite=overwrite_in,create=create_in)"
    Assets:Array of Object References=@n1
    collection:String=@n1.Collection
    sharetype_in=@n1.CollectionType
    overwrite_in=@n1.Overwrite?
    create_in=@n1.create?
   ) -> (then, ReturnValue, results=~) {inputs [Assets,collection,sharetype_in,overwrite_in,create_in], outputs [results]}
   > Return Node(Success?=@n4.results)
  ]}
]
variables []
